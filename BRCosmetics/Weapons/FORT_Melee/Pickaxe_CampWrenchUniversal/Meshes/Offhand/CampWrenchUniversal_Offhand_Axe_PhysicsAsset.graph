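# Physics asset for the offhand axe mesh of the CampWrenchUniversal pickaxe, a melee weapon cosmetic.
# It defines a single body on the b_AR_Root bone, shaped as one capsule that is offset and rotated from the bone origin.
# With no constraints or other bodies, it gives the axe one simple physics shape with no articulated parts.
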PhysicsAsset /BRCosmetics/Weapons/FORT_Melee/Pickaxe_CampWrenchUniversal/Meshes/Offhand/CampWrenchUniversal_Offhand_Axe_PhysicsAsset
bodies {b_AR_Root [{shapeType=Capsule, length=12.2848, radius=12.0057, center=<21.0989 13.5897 0.7701>, rotation=<rot 0 -98.6043 -77.647>}]}
constraints []
